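# Translucent UI material for a soft radial glow, with a parameterized color, center, radius, density and alpha.
# RadialGradientExponential builds the falloff from the UVs and scales it by the alpha parameter, and InvertAlpha blends that toward its one-minus.
# The static switch PulseAnimation replaces the fixed density with a time-sine oscillation between 1 and sDensity, paced by PulseSpeed.
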Material /FortUILibrary/Styles/Border/Materials/M_BoxScaleGlow
graph {
 properties {
  bUsedWithStaticMesh=false
  blendMode=BLEND_Translucent
  materialDomain=MD_UI
  pixelDepthOffsetMode=PDOM_Legacy
  usedShadingModels=Default Lit
 }
 nodes [
  Constant()
  n1 = ScalarParameter() {defaultValue=0.5, parameterName=sUPosition}
  n2 = ScalarParameter() {defaultValue=0.5, parameterName=sVPosition}
  n3 = ScalarParameter() {defaultValue=2, parameterName=sDensity}
  n4 = LinearInterpolate(
   B=@n3
   Alpha=MaterialFunctionCall(Speed=ScalarParameter() {defaultValue=0.4, parameterName=PulseSpeed}) {materialFunction=&DebugTimeSine}.Result
  ) {constA=1}
  n5 = LinearInterpolate(
   B=ScalarParameter() {defaultValue=1, parameterName=sAlpha, sliderMax=1}
   Alpha=MaterialFunctionCall(
    UVs=TextureCoordinate()
    CenterPosition=AppendVector(A=@n1, B=@n2)
    Radius=ScalarParameter() {defaultValue=0.5, parameterName=sRadius}
    Density=StaticSwitchParameter(True=@n4, False=@n3) {parameterName=PulseAnimation}
   ) {materialFunction=&RadialGradientExponential}.RadialGradientExponential
  )
  LinearInterpolate(A=@n5, B=OneMinus(None=@n5), Alpha=ScalarParameter() {parameterName=InvertAlpha})
  VectorParameter() {parameterName=vRGB, defaultValue=<rgba 1 1 1 1>}
 ]
}
