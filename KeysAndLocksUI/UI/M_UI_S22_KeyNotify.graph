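# Translucent UI material, named for a key notification, that builds an animated flare with drifting triangle patterns.
# Three channels of one triangle texture pan over rotatable, aspect-corrected UVs, with smoothstep thresholds varying along U to form a layered mask.
# That mask blends triangle colors over a flame gradient driven by an exponential radial flare.
# A second branch combines radial density with hard-thresholded triangles for tint and opacity; the graph shows no links to material outputs.
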Material /KeysAndLocksUI/UI/M_UI_S22_KeyNotify
graph {
 properties {blendMode=BLEND_Translucent, materialDomain=MD_UI, numCustomizedUVs=1, pixelDepthOffsetMode=PDOM_Legacy, usedShadingModels=Default Lit}
 nodes [
  n1 = MaterialFunctionCall() {materialFunction=&GetUserInterfaceUV}
  n2 = MaterialFunctionCall(UVs=@n1.Normalized UV, "Rotation Angle (0-1)"=ScalarParameter() {parameterName=Rotation}) {materialFunction=&CustomRotator}
  n3 = MaterialFunctionCall(Float2=@n1.Pixel Size) {materialFunction=&BreakOutFloat2Components}
  n4 = AppendVector(A=Divide(A=@n3.R, B=Max(B=@n3.G) {constA=0.1}), B=Constant() {r=1})
  n5 = ScalarParameter() {defaultValue=0.9048, parameterName=tri-L3-Scale}
  n6 = Add(
   A=Multiply(A=@n2.Rotated Values, B=Multiply(A=@n4, B=@n5))
   B=AppendVector(A=Constant(), B=ScalarParameter() {defaultValue=0.019, parameterName=tri-L3-Offset})
  )
  n7 = ScalarParameter() {defaultValue=1.95, parameterName=ss min}
  n8 = Power(Base=ComponentMask(None=TextureCoordinate()) {r=true}) {constExponent=1.25}
  n9 = ScalarParameter() {defaultValue=0.35, parameterName=ss max}
  n10 = MaterialFunctionCall(
   Alpha=TextureSample(UVs=Panner(Coordinate=@n6) {speedX=-0.35}) {samplerType=SAMPLERTYPE_LinearColor, texture=/Game/UI/Foundation/Items/Textures/T-Card-Hero-Anim-Triangle}.B
   Min=LinearInterpolate(B=@n7, Alpha=@n8) {constA=0.25, constB=0.8}
   Max=LinearInterpolate(A=@n9, Alpha=@n8) {constA=0.2, constB=2}
  ) {materialFunction=&SmoothStep}
  n11 = Multiply(A=@n2.Rotated Values, B=Multiply(A=@n4, B=ScalarParameter() {defaultValue=0.5743, parameterName=tri-L2-Scale}))
  n12 = AppendVector(A=Constant(), B=ScalarParameter() {defaultValue=-0.6765, parameterName=tri-L2-Offset})
  n13 = VectorParameter() {parameterName=PannerSpeed, defaultValue=<rgba -0.5 0 -0.45 0>}
  n14 = TextureSample(UVs=Panner(Coordinate=Add(A=@n11, B=@n12), Speed=AppendVector(A=@n13.B, B=@n13.B)) {speedX=-0.45}) {samplerType=SAMPLERTYPE_LinearColor, texture=&T-Card-Hero-Anim-Triangle}
  n15 = LinearInterpolate(B=@n7, Alpha=@n8) {constA=0.25, constB=0.9}
  n16 = LinearInterpolate(A=@n9, Alpha=@n8) {constA=0.1, constB=2}
  n17 = Multiply(A=@n4, B=ScalarParameter() {defaultValue=0.33, parameterName=tri-L1-Scale})
  n18 = Constant()
  n19 = ScalarParameter() {defaultValue=0.2476, parameterName=tri-L1-Offset}
  n20 = VectorParameter() {parameterName=PannerSpeed, defaultValue=<rgba -0.5 0 -0.45 0>}
  n21 = Panner(Coordinate=Add(A=Multiply(A=@n2.Rotated Values, B=@n17), B=AppendVector(A=@n18, B=@n19)), Speed=AppendVector(A=@n20.R, B=@n20.R)) {speedX=-0.5}
  n22 = LinearInterpolate(
   A=LinearInterpolate(
    A=Multiply(A=@n10.Result) {constB=0.125}
    Alpha=MaterialFunctionCall(Alpha=@n14.G, Min=@n15, Max=@n16) {materialFunction=&SmoothStep}.Result
   ) {constB=0.25}
   Alpha=MaterialFunctionCall(
    Alpha=TextureSample(UVs=@n21) {samplerType=SAMPLERTYPE_LinearColor, texture=&T-Card-Hero-Anim-Triangle}.R
    Min=LinearInterpolate(B=@n7, Alpha=@n8) {constA=0.25, constB=0.9}
    Max=LinearInterpolate(A=@n9, Alpha=@n8) {constA=0.1, constB=2}
   ) {materialFunction=&SmoothStep}.Result
  )
  n23 = TextureCoordinate() {uTiling=1.5, vTiling=1.5}
  n24 = Constant2Vector() {g=-0.5}
  n25 = MaterialFunctionCall(
   UVs=MaterialFunctionCall(UVs=Add(A=@n23, B=@n24), Texture Scale=Constant2Vector() {g=1, r=1.65}) {materialFunction=&ScaleUVsByCenter}.UVs
   CenterPosition=VectorParameter() {parameterName=FlareCenter, defaultValue=<rgba 0 0.25 0 1>}.RGB
   Radius=ScalarParameter() {defaultValue=0.85, parameterName=FlareRadius}
   Density=ScalarParameter() {defaultValue=1.75, parameterName=FlareDensity}
  ) {materialFunction=&RadialGradientExponential}
  Add(A=@n25.RadialGradientExponential, B=@n22)
  ComponentMask(None=TextureCoordinate()) {r=true}
  ComponentMask(None=TextureCoordinate()) {r=true}
  Constant2Vector() {g=1, r=1}
  Constant2Vector() {g=1.25, r=1.25}
  Constant2Vector() {g=1.5, r=1.5}
  Constant2Vector() {g=1, r=1}
  n26 = VectorParameter() {parameterName=RGB_OuterFlame, defaultValue=<rgba 1 0.5643 0 1>}
  n27 = VectorParameter() {parameterName=RGB_InnerFlame, defaultValue=<rgba 1 0.8632 0.3715 1>}
  n28 = Power(Base=@n25.RadialGradientExponential) {constExponent=0.85}
  n29 = ComponentMask(None=TextureCoordinate()) {g=true}
  LinearInterpolate(
   A=LinearInterpolate(
    A=LinearInterpolate(A=@n26.RGB, B=@n27.RGB, Alpha=@n28)
    B=Constant3Vector() {constant=<rgba 1 0.9432 0.8767 0>}
    Alpha=Power(Base=@n25.RadialGradientExponential) {constExponent=4}
   )
   B=LinearInterpolate(
    A=VectorParameter() {parameterName=Triangle_RGB_B, defaultValue=<rgba 1 0.9165 0.6661 1>}.RGB
    B=VectorParameter() {parameterName=Triangle_RGB_A, defaultValue=<rgba 1 0.8022 0 1>}.RGB
    Alpha=Power(Base=@n29) {constExponent=0.35}
   )
   Alpha=@n22
  )
  n30 = Distance(
   A=MaterialFunctionCall() {materialFunction=&ScaleUVsByCenter}.UVs
   B=Constant2Vector() {g=0.5, r=0.5}
  )
  n31 = OneMinus(None=Divide(A=Multiply(A=@n30), B=ScalarParameter() {defaultValue=0.5, parameterName=Radius, sliderMax=1}))
  n32 = ScalarParameter() {defaultValue=1.5, parameterName=Density, sliderMax=4}
  n33 = ComponentMask(None=OneMinus(None=MaterialFunctionCall!(Depth=@n31, Density=@n32, Use Exp2) {materialFunction=/Engine/Functions/Engine_MaterialFunctions01/Density/ExponentialDensity}.Result)) {r=true}
  n34 = Panner(Coordinate=MaterialFunctionCall() {materialFunction=&ScaleUVsByCenter}.UVs) {speedX=-0.035}
  n35 = TextureCoordinate()
  n36 = LinearInterpolate(B=ComponentMask(None=Sine(None=@n35) {period=2}) {r=true}, Alpha=OneMinus(None=ComponentMask(None=@n35) {r=true}))
  n37 = LinearInterpolate(Alpha=@n36) {constA=1, constB=0}
  n38 = LinearInterpolate(Alpha=@n36) {constA=3, constB=24}
  n39 = Multiply(
   A=Subtract(
    A=TextureSample(UVs=@n34) {samplerType=SAMPLERTYPE_LinearColor, texture=&T-Card-Hero-Anim-Triangle}.R
    B=@n37
   )
   B=@n38
  ) {constB=104}
  n40 = MaterialFunctionCall() {materialFunction=&ScaleUVsByCenter}
  n41 = Subtract(
   A=TextureSample(UVs=Panner(Coordinate=@n40.UVs) {speedX=-0.055}) {samplerType=SAMPLERTYPE_LinearColor, texture=&T-Card-Hero-Anim-Triangle}.G
   B=@n37
  )
  n42 = LinearInterpolate(A=LinearInterpolate(Alpha=Clamp(None=@n39) {maxDefault=0.125}), Alpha=Clamp(None=Multiply(A=@n41, B=@n38) {constB=104}) {maxDefault=0.25})
  n43 = TextureSample(UVs=Panner(Coordinate=MaterialFunctionCall() {materialFunction=&ScaleUVsByCenter}.UVs) {speedX=-0.1}) {samplerType=SAMPLERTYPE_LinearColor, texture=&T-Card-Hero-Anim-Triangle}
  n44 = Clamp(None=Multiply(A=Subtract(A=@n43.B, B=@n37), B=@n38) {constB=104})
  n45 = Sine(None=TextureCoordinate()) {period=2}
  n46 = LinearInterpolate(B=Clamp(None=LinearInterpolate(A=@n42, Alpha=@n44)) {maxDefault=0.75}, Alpha=Clamp(None=ComponentMask(None=@n45) {r=true}))
  n47 = Clamp(None=Add(A=@n33, B=DotProduct(A=@n33, B=@n46)))
  LinearInterpolate(
   A=VectorParameter() {parameterName=Col_01, defaultValue=<rgba 0.7671 1 0.7066 1>}.RGB
   B=VectorParameter() {parameterName=Col_02, defaultValue=<rgba 0.8685 1 0.6572 1>}.RGB
   Alpha=@n47
  )
  MaterialFunctionCall(UVs=TextureCoordinate(), "Rotation Angle (0-1)"=ScalarParameter() {parameterName=Rotation}) {materialFunction=&CustomRotator}
  Multiply(A=@n47, B=ScalarParameter() {defaultValue=1, parameterName=Opacity})
  TextureCoordinate()
  TextureCoordinate() {uTiling=2}
  TextureCoordinate() {uTiling=2}
  TextureCoordinate() {uTiling=2}
 ]
}
