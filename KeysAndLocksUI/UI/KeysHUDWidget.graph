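# HUD widget for the Keys and Locks feature that shows how many keys the owning player holds, as a key icon with a numeric count.
# On Construct it binds to inventory world-item changes and runs OnWorldItemsChanged, which queries the key item quantity from the player controller.
# Any change in count, up or down, updates the text and plays a scale-and-glow pickup animation.
# The widget is visible when the count is above zero and collapsed otherwise.
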WidgetBlueprint /KeysAndLocksUI/UI/KeysHUDWidget
animations [a_KeyPickup {
  binding_count=3
  length_seconds=0.5118
  object=&self:a_KeyPickup
  sequence {bindings [
    PickupHighlight {
     id=#1
     possessed_class=Image
     tracks [
      {
       type=MovieSceneColorTrack
       display_name=Color and Opacity
       sections [{
         type=MovieSceneColorSection
         end_frame=5
         start_frame=0
         channels [
          Color.R {type=MovieSceneScriptingFloatChannel, const [1,1]}
          Color.G {type=MovieSceneScriptingFloatChannel, const [1,1]}
          Color.B {type=MovieSceneScriptingFloatChannel, const [1,1]}
          Color.A {type=MovieSceneScriptingFloatChannel, keys [[0,0],[1.4863,0.6],[5.0017,0.4]]}
         ]
        }]
      }
      {
       type=MovieScene2DTransformTrack
       sections [{
         type=MovieScene2DTransformSection
         end_frame=7
         start_frame=0
         channels [
          Translation.X {type=MovieSceneScriptingFloatChannel, const [0,1]}
          Translation.Y {type=MovieSceneScriptingFloatChannel, const [-13.4752,1]}
          Angle {type=MovieSceneScriptingFloatChannel, const [0,1]}
          Scale.X {type=MovieSceneScriptingFloatChannel, keys [[0.9923,1],[5.0017,1.5],[6.3347,0.4998],[7.9707,0]]}
          Scale.Y {type=MovieSceneScriptingFloatChannel, keys [[0.9923,0.5],[5.0017,6],[7.9707,0]]}
          Shear.X {type=MovieSceneScriptingFloatChannel, const [0,1]}
          Shear.Y {type=MovieSceneScriptingFloatChannel, const [0,1]}
         ]
        }]
      }
     ]
    }
    HorizontalBox_0 {
     id=#2
     possessed_class=HorizontalBox
     tracks [{
       type=MovieScene2DTransformTrack
       sections [{
         type=MovieScene2DTransformSection
         end_frame=7
         start_frame=0
         channels [
          Translation.X {type=MovieSceneScriptingFloatChannel, const [0,1]}
          Translation.Y {type=MovieSceneScriptingFloatChannel, const [0,1]}
          Angle {type=MovieSceneScriptingFloatChannel, const [0,1]}
          Scale.X {type=MovieSceneScriptingFloatChannel, keys [[0,1],[4.9497,1.3],[7.9707,1]]}
          Scale.Y {type=MovieSceneScriptingFloatChannel, keys [[0,1],[4.9497,1.3],[7.9707,1]]}
          Shear.X {type=MovieSceneScriptingFloatChannel, const [0,1]}
          Shear.Y {type=MovieSceneScriptingFloatChannel, const [0,1]}
         ]
        }]
      }]
    }
    PickupMaterial {
     id=#3
     possessed_class=Image
     tracks [
      {
       type=MovieSceneColorTrack
       display_name=Color and Opacity
       sections [{
         type=MovieSceneColorSection
         end_frame=10
         start_frame=0
         channels [
          Color.R {type=MovieSceneScriptingFloatChannel, const [1,1]}
          Color.G {type=MovieSceneScriptingFloatChannel, const [1,1]}
          Color.B {type=MovieSceneScriptingFloatChannel, const [1,1]}
          Color.A {type=MovieSceneScriptingFloatChannel, keys [[0,1],[10.08,0]]}
         ]
        }]
      }
      {
       type=MovieScene2DTransformTrack
       sections [{
         type=MovieScene2DTransformSection
         end_frame=10
         start_frame=0
         channels [
          Translation.X {type=MovieSceneScriptingFloatChannel, const [0,1]}
          Translation.Y {type=MovieSceneScriptingFloatChannel, const [0,1]}
          Angle {type=MovieSceneScriptingFloatChannel, const [180,1]}
          Scale.X {type=MovieSceneScriptingFloatChannel, keys [[0,1],[4.9497,8],[10.08,10]]}
          Scale.Y {type=MovieSceneScriptingFloatChannel, keys [[0,1],[4.9497,4],[10.08,3]]}
          Shear.X {type=MovieSceneScriptingFloatChannel, const [0,1]}
          Shear.Y {type=MovieSceneScriptingFloatChannel, const [0,1]}
         ]
        }]
      }
     ]
    }
   ]}
 }]
graphs [
 EventGraph {nodes [
   Event Construct()
   > Parent: Construct()
   > Bind Event to On World Items Changed(
    self:Fort Inventory Context Object Reference=Get FortInventoryContext!() -> (ReturnValue:Fort Inventory Context Object Reference) {is=title K2Node_GetBlueprintContextEx}
    Delegate=Create Event()
   ) {is=title K2Node_AddDelegate}
   > OnWorldItemsChanged()
   > Set CurrentKeyCount(CurrentKeyCount:int=Get CurrentKeyCount() -> (CurrentKeyCount:int=0)) -> (Output_Get:int=0)
  ]}
 OnWorldItemsChanged {nodes [
   OnWorldItemsChanged() {is=title K2Node_FunctionEntry}
   > n1 = Cast To Athena_PlayerController!(execute, Object=GetOwningPlayer()) -> (then, CastFailed, AsAthena Player Controller:Athena Player Controller Object Reference) {is=title K2Node_DynamicCast}
   > n2 = Set NewKeyCount(NewKeyCount:int=Get Item Quantity On Player!(
     PlayerController:Controller Object Reference=@n1
     ItemDefinition:Item Definition Base Object Reference=Get KeyItemDefinition() -> (KeyItemDefinition:Fort Item Definition Object Reference)
     ItemVariantGuid:Guid Structure
    ) -> (ReturnValue:int=0) {is=title K2Node_CallFunction}) -> (Output_Get:int=0)
   > n3 = Sequence()
   > Branch(Condition="Not Equal (Integer)"(A=@n2, B=Get CurrentKeyCount() -> (CurrentKeyCount=0)))
   > Set CurrentKeyCount(CurrentKeyCount=Get NewKeyCount() -> (NewKeyCount:int=0)) -> (Output_Get=0)
   > SetCurrentValue!(
    execute
    self:Common Numeric Text Block Object Reference=Get NumericText_ResourceCount() -> (NumericText_ResourceCount:Common Numeric Text Block Object Reference)
    NewValue:float="To Float (Integer)"(InInt=Get CurrentKeyCount() -> (CurrentKeyCount=0))
   ) -> (then)
   > PlayAnimation(InAnimation=Get a_KeyPickup() -> (a_KeyPickup:Widget Animation Object Reference))
   n4 = Branch(execute=@n3.then_1, Condition="integer > integer"(A=Get CurrentKeyCount() -> (CurrentKeyCount=0)))
   > SetVisibility(InVisibility=Visible)
   SetVisibility(execute=@n4.else, InVisibility=Collapsed)
  ]}
]
variables [
 bShouldBeVisible {type=boolean}
 CurrentKeyCount {type=integer}
 KeyItemDefinition {type=FortItemDefinition, default=/KeysAndLocks/Items/KeysAndLocks/AGID_KeysAndLocks_Key}
 MaxKeys {type="array<integer>"}
]
widgets {
 info {inheritedWidgetCount=1, parentClass=&FortHUDElementWidget, rootWidgetClass=&Overlay, widgetCount=9}
 widgets [
  Overlay_16 {widgetClassPath=&Overlay, widget=Overlay_16}
  PickupMaterial {
   widgetClassPath=&Image
   bIsVariable=true
   parent=Overlay_16
   widget=PickupMaterial
   properties {
    brush {
     bIsDynamicallyLoaded=false
     drawAs=Image
     imageType=NoImage
     mirroring=NoMirror
     resourceName=None
     resourceObject=&/M_UI_S22_KeyNotify
     tiling=NoTile
     imageSize=<32 32>
     margin=<ltrb 0 0 0 0>
     outlineSettings {
      bUseBrushTransparency=false
      roundingType=HalfHeightRadius
      width=0
      color {colorUseRule=UseColor_Specified, specifiedColor=<rgba 0 0 0 0>}
      cornerRadii=<0 0 0 0>
     }
     tintColor {colorUseRule=UseColor_Specified, specifiedColor=<rgba 1 1 1 1>}
     uVRegion {bIsValid=false, max=<0 0>, min=<0 0>}
    }
    renderTransform {angle=180, scale=<1 1>, shear=<0 0>, translation=<0 0>}
   }
   slot {class=OverlaySlot, properties {horizontalAlignment=HAlign_Fill, verticalAlignment=VAlign_Fill, padding=<ltrb 0 0 0 20>}}
  }
  PickupHighlight {
   widgetClassPath=&Image
   bIsVariable=true
   parent=Overlay_16
   widget=PickupHighlight
   properties {brush {
     bIsDynamicallyLoaded=false
     drawAs=Image
     imageType=NoImage
     mirroring=NoMirror
     resourceName=None
     resourceObject=/FortUILibrary/Styles/Border/Materials/M_BoxScaleGlow
     tiling=NoTile
     imageSize=<32 32>
     margin=<ltrb 0 0 0 0>
     outlineSettings {
      bUseBrushTransparency=false
      roundingType=HalfHeightRadius
      width=0
      color {colorUseRule=UseColor_Specified, specifiedColor=<rgba 0 0 0 0>}
      cornerRadii=<0 0 0 0>
     }
     tintColor {colorUseRule=UseColor_Specified, specifiedColor=<rgba 0.855 0.9911 0.5149 1>}
     uVRegion {bIsValid=false, max=<0 0>, min=<0 0>}
    }}
   slot {class=OverlaySlot, properties {horizontalAlignment=HAlign_Fill, verticalAlignment=VAlign_Fill}}
  }
  HorizontalBox_0 {
   widgetClassPath=&HorizontalBox
   parent=Overlay_16
   widget=HorizontalBox_0
   slot {class=OverlaySlot, properties {horizontalAlignment=HAlign_Right, verticalAlignment=VAlign_Center, padding=<ltrb 0 0 4 0>}}
  }
  Overlay_0 {
   widgetClassPath=&Overlay
   parent=HorizontalBox_0
   widget=Overlay_0
   slot {class=HorizontalBoxSlot, properties {padding=<ltrb 0 0 0 -5>}}
  }
  Icon_dropshadow {
   widgetClassPath=&Image
   parent=Overlay_0
   widget=Icon_dropshadow
   properties {
    visibility=SelfHitTestInvisible
    brush {
     bIsDynamicallyLoaded=false
     drawAs=Image
     imageType=NoImage
     mirroring=NoMirror
     resourceName=None
     resourceObject=&/T_UI_ICON_S22_Key
     tiling=NoTile
     imageSize=<60 60>
     margin=<ltrb 0 0 0 0>
     outlineSettings {
      bUseBrushTransparency=false
      roundingType=HalfHeightRadius
      width=0
      color {colorUseRule=UseColor_Specified, specifiedColor=<rgba 0 0 0 0>}
      cornerRadii=<0 0 0 0>
     }
     tintColor {colorUseRule=UseColor_Specified, specifiedColor=<rgba 0 0 0 1>}
     uVRegion {bIsValid=false, max=<0 0>, min=<0 0>}
    }
    colorAndOpacity=<rgba 0 0 0 0.8>
    renderTransform {angle=0, scale=<1 1>, shear=<0 0>, translation=<1.7 1.7>}
   }
   slot {class=OverlaySlot}
  }
  Icon {
   widgetClassPath=&Image
   parent=Overlay_0
   widget=Icon
   properties {
    visibility=SelfHitTestInvisible
    brush {
     bIsDynamicallyLoaded=false
     drawAs=Image
     imageType=NoImage
     mirroring=NoMirror
     resourceName=None
     resourceObject=&T_UI_ICON_S22_Key
     tiling=NoTile
     imageSize=<60 60>
     margin=<ltrb 0 0 0 0>
     outlineSettings {
      bUseBrushTransparency=false
      roundingType=HalfHeightRadius
      width=0
      color {colorUseRule=UseColor_Specified, specifiedColor=<rgba 0 0 0 0>}
      cornerRadii=<0 0 0 0>
     }
     tintColor {colorUseRule=UseColor_Specified, specifiedColor=<rgba 1 1 1 1>}
     uVRegion {bIsValid=false, max=<0 0>, min=<0 0>}
    }
   }
   slot {class=OverlaySlot}
  }
  NumericText_ResourceCount {
   widgetClassPath=/Script/CommonUI.CommonNumericTextBlock
   bIsVariable=true
   parent=HorizontalBox_0
   widget=NumericText_ResourceCount
   properties {
    text=0
    font {
     bForceMonospaced=false
     bMaterialIsStencil=false
     fontMaterial=null
     fontObject=&BurbankSmall
     letterSpacing=0
     monospacedWidth=1
     size=16
     skewAmount=0
     typefaceFontName=Bold
     outlineSettings {
      bApplyOutlineToDropShadows=false
      bMiteredCorners=false
      bSeparateFillAlpha=true
      outlineMaterial=&M-FontShadow
      outlineSize=3
      outlineColor=<rgba 0 0 0 0.5>
     }
    }
    formattingSpecification {
     alwaysSign=false
     maximumFractionalDigits=0
     maximumIntegralDigits=324
     minimumFractionalDigits=0
     minimumIntegralDigits=1
     roundingMode=HalfFromZero
     useGrouping=false
    }
    shadowColorAndOpacity=<rgba 0 0 0 0.5>
   }
   slot {class=HorizontalBoxSlot, properties {horizontalAlignment=HAlign_Right, verticalAlignment=VAlign_Bottom, padding=<ltrb -20 0 0 -10>}}
  }
  TutorialHighlight {widgetClassPath=&TutorialHighlight}
 ]
}
