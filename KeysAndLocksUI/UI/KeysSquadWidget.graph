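# Squad UI widget that shows a small key icon for a player's key status, built on KeyPlayerInfoWidget.
# It starts collapsed on Construct, and OnKeyStatusActivated makes it hit-test-invisible only when the status is active and the player state is valid, otherwise collapsing it.
# A fixed-size box holds an overlay that layers the key icon over a darkened, slightly offset copy of the same texture as a drop shadow.
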WidgetBlueprint /KeysAndLocksUI/UI/KeysSquadWidget
graphs [EventGraph {nodes [
   n1 = Event OnKeyStatusActivated!() -> (OutputDelegate, then, InPlayerState:Fort Player State Athena Object Reference, bActivated:bool) {is=title K2Node_Event}
   > n2 = Branch(Condition=AND Boolean(A=@n1.bActivated, B=IsValid(Object=@n1.InPlayerState)))
   > SetVisibility(InVisibility=HitTestInvisible)
   SetVisibility(execute=@n2.else, InVisibility=Collapsed)
   Event Construct()
   > SetVisibility(InVisibility=Collapsed)
  ]}]
variables []
widgets {
 info {inheritedWidgetCount=1, parentClass=/Script/KeysAndLocksUI.KeyPlayerInfoWidget, rootWidgetClass=&SizeBox, widgetCount=5}
 widgets [
  SizeBox_0 {
   widgetClassPath=&SizeBox
   widget=SizeBox_0
   properties {bOverride_HeightOverride=true, bOverride_WidthOverride=true, heightOverride=20, widthOverride=20}
  }
  Overlay_16 {widgetClassPath=&Overlay, parent=SizeBox_0, widget=Overlay_16, slot {class=SizeBoxSlot}}
  Icon_dropshadow {
   widgetClassPath=&Image
   parent=Overlay_16
   widget=Icon_dropshadow
   properties {
    visibility=SelfHitTestInvisible
    brush {
     bIsDynamicallyLoaded=false
     drawAs=Image
     imageType=NoImage
     mirroring=NoMirror
     resourceName=None
     resourceObject=&/T_UI_ICON_S22_Key
     tiling=NoTile
     imageSize=<60 60>
     margin=<ltrb 0 0 0 0>
     outlineSettings {
      bUseBrushTransparency=false
      roundingType=HalfHeightRadius
      width=0
      color {colorUseRule=UseColor_Specified, specifiedColor=<rgba 0 0 0 0>}
      cornerRadii=<0 0 0 0>
     }
     tintColor {colorUseRule=UseColor_Specified, specifiedColor=<rgba 0 0 0 1>}
     uVRegion {bIsValid=false, max=<0 0>, min=<0 0>}
    }
    colorAndOpacity=<rgba 0 0 0 0.8>
    renderTransform {angle=0, scale=<1 1>, shear=<0 0>, translation=<1 1>}
   }
   slot {class=OverlaySlot}
  }
  Icon {
   widgetClassPath=&Image
   parent=Overlay_16
   widget=Icon
   properties {
    visibility=SelfHitTestInvisible
    brush {
     bIsDynamicallyLoaded=false
     drawAs=Image
     imageType=NoImage
     mirroring=NoMirror
     resourceName=None
     resourceObject=&T_UI_ICON_S22_Key
     tiling=NoTile
     imageSize=<50 50>
     margin=<ltrb 0 0 0 0>
     outlineSettings {
      bUseBrushTransparency=false
      roundingType=HalfHeightRadius
      width=0
      color {colorUseRule=UseColor_Specified, specifiedColor=<rgba 0 0 0 0>}
      cornerRadii=<0 0 0 0>
     }
     tintColor {colorUseRule=UseColor_Specified, specifiedColor=<rgba 1 1 1 1>}
     uVRegion {bIsValid=false, max=<0 0>, min=<0 0>}
    }
   }
   slot {class=OverlaySlot, properties {horizontalAlignment=HAlign_Fill, verticalAlignment=VAlign_Fill}}
  }
  TutorialHighlight {widgetClassPath=&TutorialHighlight}
 ]
}
